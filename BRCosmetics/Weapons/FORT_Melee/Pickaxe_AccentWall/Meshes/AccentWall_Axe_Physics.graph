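# Physics asset for the AccentWall pickaxe, defining capsule bodies and joints for rope pieces on its guard and handle.
# Small capsules on base_rope bones anchor dyn_rope capsules through constraints with zero twist limits.
# The guard rope is a four-link chain whose swing limits tighten toward the tip.
# Handle ropes are short one- or two-link strands, mostly with narrow swing limits.
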PhysicsAsset /BRCosmetics/Weapons/FORT_Melee/Pickaxe_AccentWall/Meshes/AccentWall_Axe_Physics
bodies {
 base_ropeGuard [{shapeType=Capsule, length=0.505, radius=0.505, shapeName=base_ropeGuard_capsule, rotation=<rot 0 -90 0>}]
 base_ropeHandle_a [{shapeType=Capsule, length=0.505, radius=0.505, shapeName=base_ropeHandle_a_capsule, rotation=<rot 0 -90 0>}]
 base_ropeHandle_b [{shapeType=Capsule, length=0.505, radius=0.505, shapeName=base_ropeHandle_b_capsule, rotation=<rot 0 -90 0>}]
 base_ropeHandle_ba [{shapeType=Capsule, length=0.505, radius=0.505, shapeName=base_ropeHandle_ba_capsule, rotation=<rot 0 -90 0>}]
 base_ropeHandle_bb [{shapeType=Capsule, length=0.505, radius=0.505, shapeName=base_ropeHandle_bb_capsule, rotation=<rot 0 -90 0>}]
 base_ropeHandle_c [{shapeType=Capsule, length=0.505, radius=0.505, shapeName=base_ropeHandle_c_capsule, rotation=<rot 0 -90 0>}]
 dyn_ropeGuard_1 [{shapeType=Capsule, length=4, radius=0.8, shapeName=dyn_ropeGuard_1_capsule, center=<2.6741 0.0098 -0.082>, rotation=<rot 0 89.5919 -88.1571>}]
 dyn_ropeGuard_2 [{
   shapeType=Capsule
   length=7.5
   radius=0.8
   shapeName=dyn_ropeGuard_2_capsule
   center=<4.6208 -0.0395 0.0116>
   rotation=<rot 0 -89.5325 -89.8481>
  }]
 dyn_ropeGuard_3 [{
   shapeType=Capsule
   length=9.6381
   radius=0.8
   shapeName=dyn_ropeGuard_3_capsule
   center=<5.7346 -0.0363 -0.0586>
   rotation=<rot 0 90.3536 -89.4214>
  }]
 dyn_ropeGuard_4 [{
   shapeType=Capsule
   length=2.5
   radius=0.8
   shapeName=dyn_ropeGuard_4_capsule
   center=<1.9002 -0.0031 -0.0021>
   rotation=<rot 0 -91.2589 -89.159>
  }]
 dyn_ropeHandle_a_1 [{
   shapeType=Capsule
   length=2.3941
   radius=1.3294
   shapeName=dyn_ropeHandle_a_1_capsule
   center=<4.4793 -0.3155 -0.3167>
   rotation=<rot 0 -110.9078 -78.5672>
  }]
 dyn_ropeHandle_b_1 [{
   shapeType=Capsule
   length=2.528
   radius=0.8497
   shapeName=dyn_ropeHandle_b_1_capsule
   center=<2.6672 -0.1026 0.1569>
   rotation=<rot 0 -176.1789 -70.0252>
  }]
 dyn_ropeHandle_ba_1 [{
   shapeType=Capsule
   length=0.8083
   radius=0.4041
   shapeName=dyn_ropeHandle_ba_1_capsule
   center=<1.7315 0.0515 -0.056>
   rotation=<rot 0 -101.8572 -78.5077>
  }]
 dyn_ropeHandle_bb_1 [{
   shapeType=Capsule
   length=0.505
   radius=0.505
   shapeName=dyn_ropeHandle_bb_1_capsule
   center=<1.6047 0.0072 0.0448>
   rotation=<rot 0 -75.496 -85.4915>
  }]
 dyn_ropeHandle_c_1 [{
   shapeType=Capsule
   length=2.2
   radius=0.8
   shapeName=dyn_ropeHandle_c_1_capsule
   center=<2.2749 0.317 -0.0204>
   rotation=<rot 0 -89.3268 -88.1041>
  }]
 dyn_ropeHandle_c_2 [{
   shapeType=Capsule
   length=2
   radius=0.8
   shapeName=dyn_ropeHandle_c_2_capsule
   center=<1.9628 -0.0626 0.0064>
   rotation=<rot 0 -82.8048 -88.4146>
  }]
}
constraints [
 {bone1Name=dyn_ropeGuard_1, bone2Name=base_ropeGuard, swing1LimitDegrees=60, swing2LimitDegrees=60, twistLimitDegrees=0}
 {bone1Name=dyn_ropeGuard_2, bone2Name=dyn_ropeGuard_1, swing1LimitDegrees=35, swing2LimitDegrees=35, twistLimitDegrees=0}
 {bone1Name=dyn_ropeGuard_3, bone2Name=dyn_ropeGuard_2, swing1LimitDegrees=25, swing2LimitDegrees=25, twistLimitDegrees=0}
 {bone1Name=dyn_ropeGuard_4, bone2Name=dyn_ropeGuard_3, swing1LimitDegrees=15, swing2LimitDegrees=15, twistLimitDegrees=0}
 {bone1Name=dyn_ropeHandle_a_1, bone2Name=base_ropeHandle_a, swing1LimitDegrees=15, swing2LimitDegrees=15, twistLimitDegrees=0}
 {bone1Name=dyn_ropeHandle_b_1, bone2Name=base_ropeHandle_b, swing1LimitDegrees=15, swing2LimitDegrees=30, twistLimitDegrees=0}
 {bone1Name=dyn_ropeHandle_ba_1, bone2Name=base_ropeHandle_ba, swing1LimitDegrees=15, swing2LimitDegrees=15, twistLimitDegrees=0}
 {bone1Name=dyn_ropeHandle_bb_1, bone2Name=base_ropeHandle_bb, swing1LimitDegrees=15, swing2LimitDegrees=15, twistLimitDegrees=0}
 {bone1Name=dyn_ropeHandle_c_2, bone2Name=dyn_ropeHandle_c_1, swing1LimitDegrees=15, swing2LimitDegrees=15, twistLimitDegrees=0}
 {bone1Name=dyn_ropeHandle_c_1, bone2Name=base_ropeHandle_c, swing1LimitDegrees=15, swing2LimitDegrees=15, twistLimitDegrees=0}
]
