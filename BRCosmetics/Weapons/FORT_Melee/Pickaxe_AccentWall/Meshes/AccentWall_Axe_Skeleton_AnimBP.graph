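# Animation Blueprint for the AccentWall pickaxe that adds rigid-body secondary motion to its skeleton, simulated relative to b_AR_Root.
# A front-end flag selects between a fixed-setting simulation and an in-game one with dynamic gravity and sim-space settings.
# Each update computes gravity via a deprecated helper and picks sprint settings when moving or sprinting without jumping or falling, otherwise regular.
# Knee sliding additively rotates two rope-guard bones after the in-game simulation.
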AnimBlueprint /BRCosmetics/Weapons/FORT_Melee/Pickaxe_AccentWall/Meshes/AccentWall_Axe_Skeleton_AnimBP
target_skeleton=&/AccentWall_Axe_Skeleton
graphs [
 AnimGraph {nodes [
   n1 = RigidBody!(ComponentPose=Local To Component(LocalPose=Input Pose() {name=InPose_0}), Alpha) -> (Pose) {
    overridePhysicsAsset=&/AccentWall_Axe_Physics
    simulationSpace=BaseBoneSpace
    alphaBoolBlend {blendOption=Linear}
    baseBoneRef {boneName=b_AR_Root}
    componentAppliedLinearAccClamp=<3000 3000 3000>
    componentLinearAccScale=<0.3 0.4 0.3>
    componentLinearVelScale=<0.3 0.4 0.3>
    simSpaceSettings {maxAngularVelocity=3, maxLinearAcceleration=700, maxLinearVelocity=500, worldAlpha=0.65}
   }
   n2 = RigidBody!(
    ComponentPose=Local To Component(LocalPose=Input Pose())
    Alpha
    OverrideWorldGravity:Vector=Get GravityOverride Output() -> (GravityOverride Output:Vector)
    SimSpaceSettings:Sim Space Settings Structure=Get Sim Space Settings Output() -> (Sim Space Settings Output:Sim Space Settings Structure)
   ) -> (Pose) {
    bOverrideWorldGravity=true
    overridePhysicsAsset=&AccentWall_Axe_Physics
    simulationSpace=BaseBoneSpace
    alphaBoolBlend {blendOption=Linear}
    baseBoneRef {boneName=b_AR_Root}
    componentAppliedLinearAccClamp=<3000 3000 3000>
    componentLinearAccScale=<0.3 0.4 0.3>
    componentLinearVelScale=<0.3 0.4 0.3>
   }
   n3 = Get bIsKneeSliding()
   n4 = "Transform (Modify) Bone - Bone: dyn_ropeGuard_1"!(ComponentPose=@n2, bAlphaBoolEnabled:bool=@n3, Translation="0,0,0", Rotation="-0.000,52.936,0.000", Scale="1,1,1") -> (Pose) {
    alphaInputType=Bool
    rotationMode=BMM_Additive
    rotationSpace=BCS_BoneSpace
    alphaBoolBlend {blendOption=Linear}
    boneToModify {boneName=dyn_ropeGuard_1}
    rotation=<rot 0 52.936 0>
   }
   n5 = "Transform (Modify) Bone - Bone: dyn_ropeGuard_3"!(ComponentPose=@n4, bAlphaBoolEnabled:bool=@n3, Translation="0,0,0", Rotation="-0.000,23.187,0.000", Scale="1,1,1") -> (Pose) {
    alphaInputType=Bool
    rotationMode=BMM_Additive
    rotationSpace=BCS_BoneSpace
    alphaBoolBlend {blendOption=Linear}
    boneToModify {boneName=dyn_ropeGuard_3}
    rotation=<rot 0 23.187 0>
   }
   Output Pose(Result=Blend Poses by bool(BlendPose_0=Component To Local(ComponentPose=@n1), BlendPose_1=Component To Local(ComponentPose=@n5), bActiveValue=Get bIsInFrontEnd()) {blendType=Linear})
  ]}
 EventGraph {nodes [
   Event BlueprintUpdateAnimation()
   > n1 = Sequence!(execute) -> (then_0, then_1, then_2:Exec)
   > n2 = Gravity Override State Function DEPRECATED(CharacterPartAnimInstance=Self-Reference(), GameTimeInSeconds=GetGameTimeInSeconds(), GravityOverrideParams=Get Gravity Override Params())
   > Set GravityOverride Output(GravityOverride Output:Vector=@n2) -> (Output_Get:Vector)
   n3 = Get bIsJumping()
   n4 = Get bIsFalling()
   n5 = Get bIsSprinting()
   n6 = Get bIsMoving2D()
   n7 = Branch(execute=@n1.then_1, Condition=AND Boolean(A=OR Boolean(A=@n6, B=@n5), B=NOR Boolean!(A:bool=@n3, B:bool=@n4) -> (ReturnValue:bool)))
   > Set Sim Space Settings Output(Sim Space Settings Output:Sim Space Settings Structure=Get Sim Space Settings Sprint() -> (Sim Space Settings Sprint:Sim Space Settings Structure)) -> (Output_Get:Sim Space Settings Structure)
   Set Sim Space Settings Output(execute=@n7.else, Sim Space Settings Output=Get Sim Space Settings Regular() -> (Sim Space Settings Regular:Sim Space Settings Structure))
   TryGetPawnOwner()
  ]}
]
variables [
 Gravity Override Params {type=GravityOverrideParamsStruct}
 GravityOverride Output {type=Vector}
 Sim Space Settings Output {
  type=SimSpaceSettings
  default {
   dampingAlpha=1
   maxAngularAcceleration=10000
   maxAngularVelocity=3
   maxLinearAcceleration=700
   maxLinearVelocity=500
   velocityScaleZ=1
   worldAlpha=1
   externalLinearDragV=<0 0 0>
  }
 }
 Sim Space Settings Regular {
  type=SimSpaceSettings
  default {
   dampingAlpha=1
   maxAngularAcceleration=10000
   maxAngularVelocity=3
   maxLinearAcceleration=700
   maxLinearVelocity=500
   velocityScaleZ=1
   worldAlpha=0.65
   externalLinearDragV=<0 0 0>
  }
 }
 Sim Space Settings Sprint {
  type=SimSpaceSettings
  default {
   dampingAlpha=1
   maxAngularAcceleration=10000
   maxAngularVelocity=3
   maxLinearAcceleration=2000
   maxLinearVelocity=500
   velocityScaleZ=1
   worldAlpha=0.6
   externalLinearDragV=<0 0 0>
  }
 }
]
